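# Niagara hit effect for the M_MED_Informer character body cosmetic, built from three sprite emitters that each burst a single camera-offset particle.
# Checkerboard003 pops to size quickly, fades through an alpha curve and animates four dynamic material parameters over its life.
# Distortion adds a randomly rotated, very faint sprite that grows then shrinks, while Glow follows slightly later as a warm, brightened flash.
# The system exposes no user parameters.
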NiagaraSystem /BRCosmetics/Characters/Player/Male/Medium/Bodies/M_MED_Informer/Effects/NS_Informer_Hit2
dependencies {
 usedDataInterfaces [&NiagaraDataInterfaceCurve, &NiagaraDataInterfaceColorCurve, &NiagaraDataInterfaceVectorCurve]
 usedDynamicInputs []
 usedModules [&SystemState]
}
emitters {
 Checkerboard003 {
  emitterUpdateScript [
   &EmitterState(Life Cycle Mode=::NewEnumerator1, Loop Behavior=::NewEnumerator1, Loop Duration=1, Scalability Mode=::NewEnumerator1, UseLoopDelay=false)
   &SpawnRate(
    SpawnRate=&Lerp_Float(A=100, B=20, Alpha=&Power(a=$Emitter.LoopedAge, b=1))
    Use Spawn Probability=false
   ) {enabled=false}
   &SpawnBurst_Instantaneous(Spawn Count=1, Use Spawn Probability=false)
  ]
  particleSpawnScript [
   /Niagara/Modules/Spawn/Initialization/InitializeParticle(
    Color=<rgba 0.66 0.66 0.66 1>
    Lifetime=0.55
    Module.Write Color=true
    Module.Write Lifetime=true
    Module.Write Mass=false
    Module.Write Material Random=true
    Module.Write Position=false
    Module.Write Scale=false
    Module.Write SpriteRotation=false
    Module.Write SpriteSize=true
    Position=&SimulationPosition()
    Sprite Size=&Vector2DFromFloat(Value=38)
   )
   &TorusLocation(
    Align Torus With Axis=true
    Large Radius=5
    MaskSpawnGroup=false
    Module.Override Randomness=false
    Surface Position=1
    Torus Axis=<1 0 0>
    Torus Mode=::NewEnumerator2
   ) {enabled=false}
   &CameraOffset(Camera Offset Amount=25)
  ]
  particleUpdateScript [
   &ParticleState()
   &ScaleColor(
    Scale Alpha=&FloatFromCurve(FloatCurve=DataInterface() {propertyValues {Curve {keys [<at 0.0002 0>,<at 0.5039 0.5>,<at 0.8089 0.5>,<at 1.0011 0.002>]}}})
    Scale RGB=&MakeVectorFromLinearColorRGB(Color=&ColorFromCurve(ColorCurve=DataInterface() {propertyValues {
        AlphaCurve {keys []}
        BlueCurve {keys [<at -0.0023 1 RCIM_Cubic>,<at 0.4914 1>,<at 1 1>,<at 1.0654 1 RCIM_Cubic>]}
        GreenCurve {keys [<at -0.0023 1 RCIM_Cubic>,<at 0.4914 1>,<at 1 1>,<at 1.0654 1 RCIM_Cubic>]}
        RedCurve {keys [<at -0.0023 1 RCIM_Cubic>,<at 0.4914 1>,<at 1 1>,<at 1.0654 1 RCIM_Cubic>]}
       }}))
    ScaleA=true
    ScaleRGB=true
   ) {module=ScaleColor001}
   &ScaleSpriteSize(Scale Factor=&Vector2DFromFloat(Value=&FloatFromCurve(FloatCurve=DataInterface() {propertyValues {Curve {keys [<at -0.0046 0 RCIM_Cubic>, {arriveTangent=0.0019, interpMode=RCIM_Cubic, leaveTangent=0.0019, tangentMode=RCTM_User, time=0.15, value=1}]}}})))
   &DynamicMaterialParameters(
    Index 0 Param 1=&FloatFromCurve(FloatCurve=DataInterface() {propertyValues {Curve {keys [
         {arriveTangent=2.6519, interpMode=RCIM_Cubic, leaveTangent=2.6519, tangentMode=RCTM_User, time=0, value=0}
         {arriveTangent=0.002, interpMode=RCIM_Cubic, leaveTangent=0.002, time=0.7905, value=0.9875}
         {arriveTangent=-9.2038, interpMode=RCIM_Cubic, leaveTangent=-9.2038, tangentMode=RCTM_User, time=1.0011, value=0.002}
        ]}}})
    Index 0 Param 2=&FloatFromCurve(FloatCurve=DataInterface() {propertyValues {Curve {keys [
         {arriveTangent=2.124, interpMode=RCIM_Cubic, leaveTangent=2.124, tangentMode=RCTM_User, time=0, value=0}
         {arriveTangent=0.002, interpMode=RCIM_Cubic, leaveTangent=0.002, time=0.8383, value=0.77}
         {arriveTangent=-10.7378, interpMode=RCIM_Cubic, leaveTangent=-10.7378, tangentMode=RCTM_User, time=1.0011, value=0.002}
        ]}}})
    Index 0 Param 3=&FloatFromCurve(FloatCurve=DataInterface() {propertyValues {Curve {keys [
         <at -0.001 0.0052>
         {arriveTangent=0.8079, interpMode=RCIM_Cubic, leaveTangent=0.8079, tangentMode=RCTM_User, time=0.3012, value=0.1163}
         {arriveTangent=1.7452, interpMode=RCIM_Cubic, leaveTangent=1.7451, tangentMode=RCTM_User, time=1.0407, value=1.005}
        ]}}})
    Index 0 Param 4=&FloatFromCurve(FloatCurve=DataInterface() {propertyValues {Curve {keys [{interpMode=RCIM_Cubic, tangentMode=RCTM_User, time=-0.0008, value=0.0052}, <at 0.2503 -1>, <at 0.9003 -1 RCIM_Cubic>, <at 1.0011 -1>]}}})
    Module.Param0WriteEnabled=true
    Module.Param1WriteEnabled=true
    Module.Param2WriteEnabled=true
    Module.Param3WriteEnabled=true
   )
  ]
  rendererClasses [&NiagaraSpriteRendererProperties]
  renderers [{rendererClass=&NiagaraSpriteRendererProperties}]
 }
 Distortion {
  emitterUpdateScript [
   &EmitterState(Life Cycle Mode=::NewEnumerator1, Loop Behavior=::NewEnumerator1, Loop Duration=0.3, UseLoopDelay=false)
   &SpawnRate(
    SpawnRate=&Lerp_Float(A=100, B=20, Alpha=&Power(a=$Emitter.LoopedAge, b=1))
    Use Spawn Probability=false
   ) {enabled=false}
   &SpawnBurst_Instantaneous(Spawn Count=1, Use Spawn Probability=false)
  ]
  particleSpawnScript [
   &InitializeParticle(
    Lifetime=&UniformRangedFloat(Minimum=0.52, Maximum=0.52)
    Module.Write Color=true
    Module.Write Lifetime=true
    Module.Write Mass=true
    Module.Write Material Random=true
    Module.Write Position=false
    Module.Write Scale=false
    Module.Write SpriteRotation=true
    Module.Write SpriteSize=true
    Position=&SimulationPosition()
    Sprite Rotation=&NormalizedAngleToDegrees(NormalizedAngle=&UniformRangedFloat(Minimum=0, Maximum=1))
    Sprite Size=&Vector2DFromFloat(Value=&UniformRangedFloat(Minimum=80, Maximum=80))
   )
   &TorusLocation(
    Align Torus With Axis=true
    Large Radius=5
    MaskSpawnGroup=false
    Module.Override Randomness=false
    Surface Position=1
    Torus Axis=<1 0 0>
    Torus Mode=::NewEnumerator2
   ) {enabled=false}
   &CameraOffset(Camera Offset Amount=20)
  ]
  particleUpdateScript [
   &ParticleState()
   &ScaleSpriteSize(Scale Factor=&Vector2DFromFloat(Value=&FloatFromCurve(FloatCurve=DataInterface() {propertyValues {Curve {keys [
          <at -0.0046 0 RCIM_Cubic>
          {arriveTangent=0.0019, interpMode=RCIM_Cubic, leaveTangent=0.0019, tangentMode=RCTM_User, time=0.5098, value=0.9653}
          {arriveTangent=-3.7458, interpMode=RCIM_Cubic, leaveTangent=-3.7458, tangentMode=RCTM_User, time=1.0011, value=0.002}
         ]}}})))
   &ScaleColor(
    Scale Alpha=&FloatFromCurve(FloatCurve=DataInterface() {propertyValues {Curve {keys [<at 0 0 RCIM_Cubic>,<at 0.4943 0.02 RCIM_Cubic>,<at 0.8 0>]}}})
    ScaleA=true
    ScaleRGB=true
   )
  ]
  rendererClasses [&NiagaraSpriteRendererProperties]
  renderers [{rendererClass=&NiagaraSpriteRendererProperties}]
 }
 Glow {
  emitterUpdateScript [
   &EmitterState(Life Cycle Mode=::NewEnumerator1, Loop Behavior=::NewEnumerator1, Loop Duration=0.3, UseLoopDelay=false)
   &SpawnRate(
    SpawnRate=&Lerp_Float(A=100, B=20, Alpha=&Power(a=$Emitter.LoopedAge, b=1))
    Use Spawn Probability=false
   ) {enabled=false}
   &SpawnBurst_Instantaneous(Spawn Count=1, Spawn Time=0.1, Use Spawn Probability=false)
  ]
  particleSpawnScript [
   &InitializeParticle(
    Color=<rgba 1 0.8847 0.7371 0.35>
    Lifetime=&UniformRangedFloat(Minimum=0.25, Maximum=0.25)
    Module.Write Color=true
    Module.Write Lifetime=true
    Module.Write Mass=false
    Module.Write Material Random=true
    Module.Write Position=false
    Module.Write Scale=false
    Module.Write SpriteRotation=false
    Module.Write SpriteSize=true
    Position=&SimulationPosition()
    Sprite Size=&Vector2DFromFloat(Value=55)
   )
   &TorusLocation(
    Align Torus With Axis=true
    Large Radius=5
    MaskSpawnGroup=false
    Module.Override Randomness=false
    Surface Position=1
    Torus Axis=<1 0 0>
    Torus Mode=::NewEnumerator2
   ) {enabled=false}
   &CameraOffset(Camera Offset Amount=25)
  ]
  particleUpdateScript [
   &ParticleState()
   &ScaleSpriteSize(Scale Factor=&Vector2DFromFloat(Value=&FloatFromCurve(FloatCurve=DataInterface() {propertyValues {Curve {keys [
          {arriveTangent=2.8759, interpMode=RCIM_Cubic, leaveTangent=2.8759, tangentMode=RCTM_User, time=-0.0068, value=0.1019}
          {arriveTangent=0.0019, interpMode=RCIM_Cubic, leaveTangent=0.0019, tangentMode=RCTM_User, time=0.3839, value=1.0042}
          {arriveTangent=-0.0147, interpMode=RCIM_Cubic, leaveTangent=-0.0147, tangentMode=RCTM_User, time=0.9993, value=1.0021}
         ]}}})))
   &ScaleColor(
    Scale Alpha=&FloatFromCurve(FloatCurve=DataInterface() {propertyValues {Curve {keys [<at -0.0046 0 RCIM_Cubic>, {arriveTangent=0.0019, interpMode=RCIM_Cubic, leaveTangent=0.0019, time=0.4961, value=0.7268}, <at 1.0011 0.002>]}}})
    Scale RGB=&VectorFromCurve(VectorCurve=DataInterface() {propertyValues {
       XCurve {keys [<at -0.0011 3>,<at 0.5 3>,<at 0.9989 3>]}
       YCurve {keys [<at -0.0011 3>,<at 0.5 3>,<at 0.9989 3>]}
       ZCurve {keys [<at -0.0011 3>,<at 0.5 3>,<at 0.9989 3>]}
      }})
    ScaleA=true
    ScaleRGB=true
   )
  ]
  rendererClasses [&NiagaraSpriteRendererProperties]
  renderers [{rendererClass=&NiagaraSpriteRendererProperties}]
 }
}
summary {
 systemName=NS_Informer_Hit2
 emitters [
  {emitterName=Distortion, rendererClasses [&NiagaraSpriteRendererProperties]}
  {emitterName=Checkerboard003, rendererClasses [&NiagaraSpriteRendererProperties]}
  {emitterName=Glow, rendererClasses [&NiagaraSpriteRendererProperties]}
 ]
 userVariables []
}
