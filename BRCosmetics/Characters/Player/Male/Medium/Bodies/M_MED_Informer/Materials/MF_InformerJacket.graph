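# Material function for the Informer outfit's jacket that overrides base color and emissive on input material attributes, blended by a Mask input.
# A tiled grid texture times the Glitch parameter slightly warps UVs for the diffuse and emissive samples, and Glitch also boosts emissive tint.
# A time-driven SphereMask sweeps a brightness pulse along EmissivePath, added with subtle white emissive to the incoming emissive.
# The Fresnel branch does not feed the output.
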MaterialFunction /BRCosmetics/Characters/Player/Male/Medium/Bodies/M_MED_Informer/Materials/MF_InformerJacket
graph {nodes [
  n1 = GetMaterialAttributes(None=FunctionInput() {inputName=Material, inputType=FunctionInput_MaterialAttributes}) {attributeGetTypes [#1]}
  n2 = VectorParameter() {group=Informer Jacket, parameterName=Subtle White Emissive, defaultValue=<rgba 0.8177 0.8177 0.8177 1>}
  n3 = FunctionInput() {bUsePreviewValueAsDefault=true, inputName=WhiteEmissive, inputType=FunctionInput_Scalar}
  n4 = TextureCoordinate()
  n5 = Multiply(A=TextureCoordinate(), B=ScalarParameter() {defaultValue=4, group=Informer Jacket, parameterName=Hex Grid Tiling})
  n6 = ScalarParameter() {group=Informer Jacket, parameterName=Glitch}
  n7 = LinearInterpolate(
   A=@n4
   B=Multiply(A=@n4, B=ScalarParameter() {defaultValue=0.9995, group=Informer Jacket, parameterName=UV Warp Amount})
   Alpha=Multiply(
    A=TextureSample(UVs=@n5) {samplerType=SAMPLERTYPE_LinearColor, texture=/Game/Gadgets/Assets/Base_Device/Textures/BaseDevice_2_rgb_PivotPos_a_Random0-1_UV_2}.R
    B=@n6
   )
  )
  n8 = TextureSampleParameter2D(UVs=@n7) {
   group=Informer Jacket
   parameterName=New Emissive
   samplerType=SAMPLERTYPE_Masks
   texture=/BRCosmetics/Characters/Player/Male/Medium/Bodies/M_MED_Informer/Textures/T_M_MED_Informer_Body_E
  }
  n9 = VectorParameter() {group=Informer Jacket, parameterName=New Emissive Tint, defaultValue=<rgba 0 0.343 0.5 1>}
  n10 = LinearInterpolate(A=@n9.RGB, B=Multiply(A=@n9.RGB, B=ScalarParameter() {defaultValue=80, group=Informer Jacket, parameterName=Emissive Rmap}), Alpha=@n6)
  n11 = Multiply(A=@n10, B=ScalarParameter() {defaultValue=3, group=Informer Jacket, parameterName=Gradient Pulse Amount}) {constB=2.5}
  n12 = Frac(None=Multiply(A=Time(), B=ScalarParameter() {defaultValue=0.25, group=Informer Jacket, parameterName=EmissivePulseSpeed}))
  n13 = Clamp(None=SphereMask(
    A=LinearInterpolate(Alpha=@n12) {constA=-0.25, constB=1.25}
    B=FunctionInput() {inputName=EmissivePath, inputType=FunctionInput_Scalar}
    Radius=ScalarParameter() {defaultValue=0.25, group=Informer Jacket, parameterName=SpherePulse_IdleRadius}
    Hardness=Constant() {r=0.1}
   ))
  n14 = Add(A=Add(A=@n1.EmissiveColor, B=Multiply(A=@n2.RGB, B=@n3)), B=Multiply(A=@n8.R, B=LinearInterpolate(A=@n10, B=@n11, Alpha=@n13)))
  n15 = TextureSampleParameter2D(UVs=@n7) {group=Informer Jacket, parameterName=Base Diffuse, texture=&T_BlacksRGB}
  FunctionOutput(None=BlendMaterialAttributes(
    A=@n1.MaterialAttributes
    B=SetMaterialAttributes!(MaterialAttributes=@n1.MaterialAttributes, EmissiveColor=@n14, BaseColor=@n15.RGB) {attributeSetTypes [#1, #2]}
    Alpha=FunctionInput() {inputName=Mask, inputType=FunctionInput_Scalar}
   ))
  Multiply()
  n16 = FunctionInput() {bUsePreviewValueAsDefault=true, inputName=Fresnel, inputType=FunctionInput_Scalar}
  Multiply(A=OneMinus(None=Saturate(None=@n16)), B=@n14)
 ]}
