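# Master material for the male medium Informer character body, a Default Lit skeletal mesh surface built on material attributes.
# The base character PreFX function feeds MF_InformerJacket, driven by channel-mask parameters on the SkinFX texture for jacket mask, emissive path and white emissive.
# The result passes through the weapon-skin Clearcoat function, with the jacket mask as alpha, then into the shared character PostFX function alongside other PreFX outputs.
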Material /BRCosmetics/Characters/Player/Male/Medium/Bodies/M_MED_Informer/Materials/M_MED_Informer_Body_Master
graph {
 properties {
  bUseMaterialAttributes=true
  bUsedWithMeshDeformer=true
  bUsedWithMorphTargets=true
  bUsedWithSkeletalMesh=true
  pixelDepthOffsetMode=PDOM_Legacy
  usedShadingModels=Default Lit
 }
 nodes [
  n1 = MaterialFunctionCall() {materialFunction=&BaseFN_CharacterFunction_PreFX}
  n2 = TextureSampleParameter2D() {parameterName=SkinFX, texture=&T_BlacksRGB}
  n3 = ChannelMaskParameter(None=@n2.RGB) {parameterName=JacketMask}
  n4 = ChannelMaskParameter(None=@n2.RGB) {maskChannel=Blue, parameterName="Emissive  Path Channel", defaultValue=<rgba 0 0 1 0>}
  n5 = ChannelMaskParameter(None=@n2.RGBA) {maskChannel=Green, parameterName=White Emissive Channel, defaultValue=<rgba 0 1 0 0>}
  MaterialFunctionCall(
   AO=@n1.AO
   Distance Blend=@n1.AO
   Fresnel=@n1.AO
   Gmap=@n1.AO
   Material=MaterialFunctionCall!(
    Alpha=@n3.R
    In=MaterialFunctionCall!(Material=@n1.Result, EmissivePath=@n4.R, Fresnel=@n1.Result, WhiteEmissive=@n5.R, Mask=@n3.R) {materialFunction=&/MF_InformerJacket}.Result
   ) {materialFunction=/Game/Weapons/WeaponSkins/MaterialFunction/Clearcoat}.Result
  ) {materialFunction=&BaseFN_CharacterFunction_PostFX}
 ]
}
