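# Physics asset for the dynamic coat on the M_MED_Informer male player body, defining coat bone chains and lower-body collision shapes.
# Front, side, back-side, back and center-split chains of boxes use limited swing and twist, anchored through small MSTR capsules to spine_01 (front) or pelvis (others).
# Capsules cover pelvis, lower spine and legs, with only the left foot included.
# Asymmetric cross-constraints tie some neighboring panels together at their lower segments.
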PhysicsAsset /BRCosmetics/Characters/Player/Male/Medium/Bodies/M_MED_Informer/Meshes/M_MED_Informer_dyn_coat
bodies {
 calf_twist_01_l [{shapeType=Capsule, length=13.6815, radius=4.9152, center=<-0.1128 -0.642 0.0099>, rotation=<rot 85.0004 0.0001 5.0007>}]
 calf_twist_01_r [{shapeType=Capsule, length=19.0408, radius=4.9152, center=<3.3302 0.5194 -0.2915>, rotation=<rot 84.9983 0 10.0012>}]
 calf_twist_02_l [{shapeType=Capsule, length=20.7844, radius=7.0112, center=<3.8638 -2.4308 -0.5571>, rotation=<rot 90 90 95>}]
 calf_twist_02_r [{shapeType=Capsule, length=30.3918, radius=6.3643, center=<-3.9234 2.7407 0.5573>, rotation=<rot 90 15.3763 25.3765>}]
 dyn_coatBack01_l [{shapeType=Box, extentX=2, extentY=15, extentZ=10, center=<5 0 0>, rotation=<rot 90 0 0>}]
 dyn_coatBack01_r [{shapeType=Box, extentX=2, extentY=15, extentZ=10, center=<5 0 0>, rotation=<rot 90 0 0>}]
 dyn_coatBack02_l [{shapeType=Box, extentX=2, extentY=15, extentZ=10, center=<5 0 0>, rotation=<rot 90 0 0>}]
 dyn_coatBack02_r [{shapeType=Box, extentX=2, extentY=15, extentZ=10, center=<5 0 0>, rotation=<rot 90 0 0>}]
 dyn_coatBack03_l [{shapeType=Box, extentX=3, extentY=20, extentZ=12, center=<5.5 0 0>, rotation=<rot 90 0 0>}]
 dyn_coatBack03_r [{shapeType=Box, extentX=3, extentY=20, extentZ=12, center=<5.5 0 0>, rotation=<rot 90 0 0>}]
 dyn_coatBack04_l [{shapeType=Box, extentX=3, extentY=20, extentZ=12, center=<5.5 0 0>, rotation=<rot 90 0 0>}]
 dyn_coatBack04_r [{shapeType=Box, extentX=3, extentY=20, extentZ=12, center=<5.5 0 0>, rotation=<rot 90 0 0>}]
 dyn_coatBackSide01_l [{shapeType=Box, extentX=2, extentY=15, extentZ=10, center=<5 0 0>, rotation=<rot 90 0 0>}]
 dyn_coatBackSide01_r [{shapeType=Box, extentX=2, extentY=15, extentZ=10, center=<5 0 0>, rotation=<rot 90 0 0>}]
 dyn_coatBackSide02_l [{shapeType=Box, extentX=2, extentY=15, extentZ=10, center=<5 0 0>, rotation=<rot 90 0 0>}]
 dyn_coatBackSide02_r [{shapeType=Box, extentX=2, extentY=15, extentZ=10, center=<5 0 0>, rotation=<rot 90 0 0>}]
 dyn_coatBackSide03_l [{shapeType=Box, extentX=3, extentY=20, extentZ=12, center=<5.5 0 0>, rotation=<rot 90 0 0>}]
 dyn_coatBackSide03_r [{shapeType=Box, extentX=3, extentY=20, extentZ=12, center=<5.5 0 0>, rotation=<rot 90 0 0>}]
 dyn_coatBackSide04_l [{shapeType=Box, extentX=3, extentY=20, extentZ=12, center=<5.5 0 0>, rotation=<rot 90 0 0>}]
 dyn_coatBackSide04_r [{shapeType=Box, extentX=3, extentY=20, extentZ=12, center=<5.5 0 0>, rotation=<rot 90 0 0>}]
 dyn_coatBackSide_l_MSTR [{shapeType=Capsule, length=1, radius=1, rotation=<rot 0 0 90>}]
 dyn_coatBackSide_r_MSTR [{shapeType=Capsule, length=1, radius=1, rotation=<rot 0 0 90>}]
 dyn_coatBack_l_MSTR [{shapeType=Capsule, length=1, radius=1, rotation=<rot 0 0 90>}]
 dyn_coatBack_r_MSTR [{shapeType=Capsule, length=1, radius=1, rotation=<rot 0 0 90>}]
 dyn_coatFront01_l [{shapeType=Box, extentX=2, extentY=15, extentZ=11, center=<5.5 -2 0>, rotation=<rot 90 0 0>}]
 dyn_coatFront01_r [{shapeType=Box, extentX=2, extentY=15, extentZ=11, center=<5.5 2 0>, rotation=<rot 90 0 0>}]
 dyn_coatFront02_l [{shapeType=Box, extentX=3, extentY=15, extentZ=11, center=<5.5 -2 0>, rotation=<rot 90 0 0>}]
 dyn_coatFront02_r [{shapeType=Box, extentX=3, extentY=15, extentZ=11, center=<5.5 2 0>, rotation=<rot 90 0 0>}]
 dyn_coatFront03_l [{shapeType=Box, extentX=3, extentY=20, extentZ=11, center=<5 -2 0>, rotation=<rot 90 0 0>}]
 dyn_coatFront03_r [{shapeType=Box, extentX=3, extentY=20, extentZ=11, center=<5 2 0>, rotation=<rot 90 0 0>}]
 dyn_coatFront04_l [{shapeType=Box, extentX=3, extentY=20, extentZ=12, center=<5.5 -2 0>, rotation=<rot 90 0 0>}]
 dyn_coatFront04_r [{shapeType=Box, extentX=3, extentY=20, extentZ=12, center=<5.5 2 0>, rotation=<rot 90 0 0>}]
 dyn_coatFront05_l [{shapeType=Box, extentX=3, extentY=20, extentZ=12, center=<5.5 -2 0>, rotation=<rot 90 0 0>}]
 dyn_coatFront05_r [{shapeType=Box, extentX=3, extentY=20, extentZ=12, center=<5.5 2 0>, rotation=<rot 90 0 0>}]
 dyn_coatFront_l_MSTR [{shapeType=Capsule, length=1, radius=1, rotation=<rot 0 0 90>}]
 dyn_coatFront_r_MSTR [{shapeType=Capsule, length=1, radius=1, rotation=<rot 0 0 90>}]
 dyn_coatSide01_l [{shapeType=Box, extentX=2, extentY=12, extentZ=10, center=<5 0 0>, rotation=<rot 90 0 0>}]
 dyn_coatSide01_r [{shapeType=Box, extentX=2, extentY=12, extentZ=10, center=<5 0 0>, rotation=<rot 90 0 0>}]
 dyn_coatSide02_l [{shapeType=Box, extentX=2, extentY=20, extentZ=12, center=<6 0 0>, rotation=<rot 90 0 0>}]
 dyn_coatSide02_r [{shapeType=Box, extentX=2, extentY=12, extentZ=12, center=<6 0 0>, rotation=<rot 90 0 0>}]
 dyn_coatSide03_l [{shapeType=Box, extentX=3, extentY=20, extentZ=12, center=<5.5 0 0>, rotation=<rot 90 0 0>}]
 dyn_coatSide03_r [{shapeType=Box, extentX=3, extentY=20, extentZ=12, center=<5.5 0 0>, rotation=<rot 90 0 0>}]
 dyn_coatSide04_l [{shapeType=Box, extentX=3, extentY=20, extentZ=12, center=<5.5 0 0>, rotation=<rot 90 0 0>}]
 dyn_coatSide04_r [{shapeType=Box, extentX=3, extentY=20, extentZ=12, center=<5.5 0 0>, rotation=<rot 90 0 0>}]
 dyn_coatSide_l_MSTR [{shapeType=Capsule, length=1, radius=1, rotation=<rot 0 0 90>}]
 dyn_coatSide_r_MSTR [{shapeType=Capsule, length=1, radius=1, rotation=<rot 0 0 90>}]
 dyn_coatSplit01 [{shapeType=Box, extentX=2, extentY=15, extentZ=12, center=<6 0 0>, rotation=<rot 90 0 0>}]
 dyn_coatSplit02 [{shapeType=Box, extentX=2, extentY=15, extentZ=12, center=<6 0 0>, rotation=<rot 90 0 0>}]
 dyn_coatSplit_MSTR [{shapeType=Capsule, length=1, radius=1, rotation=<rot 0 0 90>}]
 foot_l [{shapeType=Capsule, length=4.2601, radius=4.2889, center=<-3.5689 -1.6323 0>, rotation=<rot 90 0 180>}]
 pelvis [{shapeType=Capsule, length=8.4146, radius=11.834, center=<-4.2477 1.7482 -0.0001>}]
 spine_01 [{shapeType=Capsule, length=5.0662, radius=12.3927, center=<0.0001 2.7144 0>}]
 thigh_l [{shapeType=Capsule, length=6.1787, radius=10.4709, center=<-5.0001 1.2876 0.9629>, rotation=<rot 90 0 0>}]
 thigh_r [{shapeType=Capsule, length=6.1787, radius=10.4709, center=<5 -1.8766 -2.321>, rotation=<rot 90 0 0>}]
 thigh_twist_01_l [{shapeType=Capsule, length=22.7653, radius=7.7347, center=<-2.8572 -0.4473 0>, rotation=<rot 90 7.125 2.125>}]
 thigh_twist_01_r [{shapeType=Capsule, length=24.5756, radius=8.1005, center=<0.3381 -0.5681 0>, rotation=<rot 90 0 -7.4586>}]
}
constraints [
 {bone1Name=dyn_coatFront02_r, bone2Name=dyn_coatFront01_r, swing1LimitDegrees=15, swing2LimitDegrees=60, twistLimitDegrees=5}
 {bone1Name=dyn_coatFront03_r, bone2Name=dyn_coatFront02_r, swing1LimitDegrees=15, swing2LimitDegrees=60, twistLimitDegrees=5}
 {bone1Name=dyn_coatFront04_r, bone2Name=dyn_coatFront03_r, swing1LimitDegrees=15, swing2LimitDegrees=30, twistLimitDegrees=5}
 {bone1Name=dyn_coatFront05_r, bone2Name=dyn_coatFront04_r, swing1LimitDegrees=15, swing2LimitDegrees=30, twistLimitDegrees=5}
 {bone1Name=dyn_coatFront02_l, bone2Name=dyn_coatFront01_l, swing1LimitDegrees=15, swing2LimitDegrees=60, twistLimitDegrees=5}
 {bone1Name=dyn_coatFront03_l, bone2Name=dyn_coatFront02_l, swing1LimitDegrees=15, swing2LimitDegrees=60, twistLimitDegrees=5}
 {bone1Name=dyn_coatFront04_l, bone2Name=dyn_coatFront03_l, swing1LimitDegrees=15, swing2LimitDegrees=30, twistLimitDegrees=5}
 {bone1Name=dyn_coatFront05_l, bone2Name=dyn_coatFront04_l, swing1LimitDegrees=15, swing2LimitDegrees=30, twistLimitDegrees=5}
 {
  bone1Name=thigh_twist_01_r
  bone2Name=thigh_r
  swing1LimitDegrees=45
  swing1Motion=Free
  swing2LimitDegrees=45
  swing2Motion=Free
  twistLimitDegrees=45
  twistMotion=Free
 }
 {
  bone1Name=thigh_l
  bone2Name=pelvis
  swing1LimitDegrees=45
  swing1Motion=Free
  swing2LimitDegrees=45
  swing2Motion=Free
  twistLimitDegrees=45
  twistMotion=Free
 }
 {
  bone1Name=thigh_r
  bone2Name=pelvis
  swing1LimitDegrees=45
  swing1Motion=Free
  swing2LimitDegrees=45
  swing2Motion=Free
  twistLimitDegrees=45
  twistMotion=Free
 }
 {
  bone1Name=spine_01
  bone2Name=pelvis
  swing1LimitDegrees=45
  swing1Motion=Free
  swing2LimitDegrees=45
  swing2Motion=Free
  twistLimitDegrees=45
  twistMotion=Free
 }
 {bone1Name=dyn_coatSide02_r, bone2Name=dyn_coatSide01_r, swing1LimitDegrees=5, swing2LimitDegrees=60, twistLimitDegrees=5}
 {bone1Name=dyn_coatSide03_r, bone2Name=dyn_coatSide02_r, swing1LimitDegrees=5, swing2LimitDegrees=30, twistLimitDegrees=5}
 {bone1Name=dyn_coatSide04_r, bone2Name=dyn_coatSide03_r, swing1LimitDegrees=5, swing2LimitDegrees=30, twistLimitDegrees=5}
 {bone1Name=dyn_coatSide02_l, bone2Name=dyn_coatSide01_l, swing1LimitDegrees=5, swing2LimitDegrees=60, twistLimitDegrees=5}
 {bone1Name=dyn_coatSide03_l, bone2Name=dyn_coatSide02_l, swing1LimitDegrees=5, swing2LimitDegrees=30, twistLimitDegrees=5}
 {bone1Name=dyn_coatSide04_l, bone2Name=dyn_coatSide03_l, swing1LimitDegrees=5, swing2LimitDegrees=30, twistLimitDegrees=5}
 {bone1Name=dyn_coatBack02_r, bone2Name=dyn_coatBack01_r, swing1LimitDegrees=5, swing2LimitDegrees=60, twistLimitDegrees=5}
 {bone1Name=dyn_coatBack03_r, bone2Name=dyn_coatBack02_r, swing1LimitDegrees=5, swing2LimitDegrees=30, twistLimitDegrees=5}
 {bone1Name=dyn_coatBack04_r, bone2Name=dyn_coatBack03_r, swing1LimitDegrees=5, swing2LimitDegrees=30, twistLimitDegrees=5}
 {bone1Name=dyn_coatBack02_l, bone2Name=dyn_coatBack01_l, swing1LimitDegrees=5, swing2LimitDegrees=60, twistLimitDegrees=5}
 {bone1Name=dyn_coatBack03_l, bone2Name=dyn_coatBack02_l, swing1LimitDegrees=5, swing2LimitDegrees=30, twistLimitDegrees=5}
 {bone1Name=dyn_coatBack04_l, bone2Name=dyn_coatBack03_l, swing1LimitDegrees=5, swing2LimitDegrees=30, twistLimitDegrees=5}
 {bone1Name=dyn_coatSplit02, bone2Name=dyn_coatSplit01, swing1LimitDegrees=5, swing2LimitDegrees=60, twistLimitDegrees=5}
 {
  bone1Name=calf_twist_02_l
  bone2Name=thigh_l
  swing1LimitDegrees=45
  swing1Motion=Free
  swing2LimitDegrees=45
  swing2Motion=Free
  twistLimitDegrees=45
  twistMotion=Free
 }
 {
  bone1Name=calf_twist_02_r
  bone2Name=thigh_r
  swing1LimitDegrees=45
  swing1Motion=Free
  swing2LimitDegrees=45
  swing2Motion=Free
  twistLimitDegrees=45
  twistMotion=Free
 }
 {
  bone1Name=calf_twist_01_l
  bone2Name=thigh_l
  swing1LimitDegrees=45
  swing1Motion=Free
  swing2LimitDegrees=45
  swing2Motion=Free
  twistLimitDegrees=45
  twistMotion=Free
 }
 {
  bone1Name=foot_l
  bone2Name=thigh_l
  swing1LimitDegrees=45
  swing1Motion=Free
  swing2LimitDegrees=45
  swing2Motion=Free
  twistLimitDegrees=45
  twistMotion=Free
 }
 {
  bone1Name=thigh_twist_01_l
  bone2Name=thigh_l
  swing1LimitDegrees=45
  swing1Motion=Free
  swing2LimitDegrees=45
  swing2Motion=Free
  twistLimitDegrees=45
  twistMotion=Free
 }
 {
  bone1Name=dyn_coatSide_l_MSTR
  bone2Name=pelvis
  swing1LimitDegrees=45
  swing1Motion=Free
  swing2LimitDegrees=45
  swing2Motion=Free
  twistLimitDegrees=45
  twistMotion=Free
 }
 {bone1Name=dyn_coatSide01_l, bone2Name=dyn_coatSide_l_MSTR, swing1LimitDegrees=5, swing2LimitDegrees=60, twistLimitDegrees=5}
 {
  bone1Name=dyn_coatBack_l_MSTR
  bone2Name=pelvis
  swing1LimitDegrees=45
  swing1Motion=Free
  swing2LimitDegrees=45
  swing2Motion=Free
  twistLimitDegrees=45
  twistMotion=Free
 }
 {bone1Name=dyn_coatBack01_l, bone2Name=dyn_coatBack_l_MSTR, swing1LimitDegrees=5, swing2LimitDegrees=60, twistLimitDegrees=5}
 {
  bone1Name=dyn_coatFront_l_MSTR
  bone2Name=spine_01
  swing1LimitDegrees=45
  swing1Motion=Free
  swing2LimitDegrees=45
  swing2Motion=Free
  twistLimitDegrees=45
  twistMotion=Free
 }
 {bone1Name=dyn_coatFront01_l, bone2Name=dyn_coatFront_l_MSTR, swing1LimitDegrees=15, swing2LimitDegrees=75, twistLimitDegrees=5}
 {
  bone1Name=dyn_coatFront_r_MSTR
  bone2Name=spine_01
  swing1LimitDegrees=45
  swing1Motion=Free
  swing2LimitDegrees=45
  swing2Motion=Free
  twistLimitDegrees=45
  twistMotion=Free
 }
 {bone1Name=dyn_coatFront01_r, bone2Name=dyn_coatFront_r_MSTR, swing1LimitDegrees=15, swing2LimitDegrees=75, twistLimitDegrees=5}
 {
  bone1Name=dyn_coatBackSide_l_MSTR
  bone2Name=pelvis
  swing1LimitDegrees=45
  swing1Motion=Free
  swing2LimitDegrees=45
  swing2Motion=Free
  twistLimitDegrees=45
  twistMotion=Free
 }
 {bone1Name=dyn_coatBackSide01_l, bone2Name=dyn_coatBackSide_l_MSTR, swing1LimitDegrees=5, swing2LimitDegrees=60, twistLimitDegrees=5}
 {bone1Name=dyn_coatBackSide02_l, bone2Name=dyn_coatBackSide01_l, swing1LimitDegrees=5, swing2LimitDegrees=60, twistLimitDegrees=5}
 {bone1Name=dyn_coatBackSide03_l, bone2Name=dyn_coatBackSide02_l, swing1LimitDegrees=5, swing2LimitDegrees=30, twistLimitDegrees=5}
 {bone1Name=dyn_coatBackSide04_l, bone2Name=dyn_coatBackSide03_l, swing1LimitDegrees=5, swing2LimitDegrees=30, twistLimitDegrees=5}
 {
  bone1Name=dyn_coatSplit_MSTR
  bone2Name=pelvis
  swing1LimitDegrees=45
  swing1Motion=Free
  swing2LimitDegrees=45
  swing2Motion=Free
  twistLimitDegrees=45
  twistMotion=Free
 }
 {bone1Name=dyn_coatSplit01, bone2Name=dyn_coatSplit_MSTR, swing1LimitDegrees=5, swing2LimitDegrees=60, twistLimitDegrees=5}
 {
  bone1Name=dyn_coatSide_r_MSTR
  bone2Name=pelvis
  swing1LimitDegrees=45
  swing1Motion=Free
  swing2LimitDegrees=45
  swing2Motion=Free
  twistLimitDegrees=45
  twistMotion=Free
 }
 {bone1Name=dyn_coatSide01_r, bone2Name=dyn_coatSide_r_MSTR, swing1LimitDegrees=5, swing2LimitDegrees=60, twistLimitDegrees=5}
 {
  bone1Name=dyn_coatBackSide_r_MSTR
  bone2Name=pelvis
  swing1LimitDegrees=45
  swing1Motion=Free
  swing2LimitDegrees=45
  swing2Motion=Free
  twistLimitDegrees=45
  twistMotion=Free
 }
 {bone1Name=dyn_coatBackSide01_r, bone2Name=dyn_coatBackSide_r_MSTR, swing1LimitDegrees=5, swing2LimitDegrees=60, twistLimitDegrees=5}
 {bone1Name=dyn_coatBackSide02_r, bone2Name=dyn_coatBackSide01_r, swing1LimitDegrees=5, swing2LimitDegrees=60, twistLimitDegrees=5}
 {bone1Name=dyn_coatBackSide03_r, bone2Name=dyn_coatBackSide02_r, swing1LimitDegrees=5, swing2LimitDegrees=30, twistLimitDegrees=5}
 {bone1Name=dyn_coatBackSide04_r, bone2Name=dyn_coatBackSide03_r, swing1LimitDegrees=5, swing2LimitDegrees=30, twistLimitDegrees=5}
 {
  bone1Name=dyn_coatBack_r_MSTR
  bone2Name=pelvis
  swing1LimitDegrees=45
  swing1Motion=Free
  swing2LimitDegrees=45
  swing2Motion=Free
  twistLimitDegrees=45
  twistMotion=Free
 }
 {bone1Name=dyn_coatBack01_r, bone2Name=dyn_coatBack_r_MSTR, swing1LimitDegrees=5, swing2LimitDegrees=60, twistLimitDegrees=5}
 {bone1Name=dyn_coatSplit02, bone2Name=dyn_coatBack02_r, swing1LimitDegrees=5, swing2LimitDegrees=10, twistLimitDegrees=5}
 {bone1Name=dyn_coatSplit02, bone2Name=dyn_coatBack02_l, swing1LimitDegrees=5, swing2LimitDegrees=10, twistLimitDegrees=5}
 {
  bone1Name=dyn_coatBackSide03_l
  bone2Name=dyn_coatSide03_l
  swing1LimitDegrees=15
  swing2LimitDegrees=30
  twistLimitDegrees=5
  twistMotion=Free
 }
 {
  bone1Name=calf_twist_01_r
  bone2Name=thigh_r
  swing1LimitDegrees=45
  swing1Motion=Free
  swing2LimitDegrees=45
  swing2Motion=Free
  twistLimitDegrees=45
  twistMotion=Free
 }
 {bone1Name=dyn_coatBack04_r, bone2Name=dyn_coatBackSide04_r, swing1LimitDegrees=10, swing2LimitDegrees=20, twistLimitDegrees=5}
 {bone1Name=dyn_coatSide04_r, bone2Name=dyn_coatBackSide04_r, swing1LimitDegrees=10, swing2LimitDegrees=20, twistLimitDegrees=5}
 {bone1Name=dyn_coatSide04_l, bone2Name=dyn_coatFront05_l, swing1LimitDegrees=5, swing2LimitDegrees=30, twistLimitDegrees=5, twistMotion=Free}
 {
  bone1Name=dyn_coatSide03_r
  bone2Name=dyn_coatFront04_r
  swing1LimitDegrees=10
  swing1Motion=Free
  swing2LimitDegrees=30
  swing2Motion=Free
  twistLimitDegrees=5
  twistMotion=Free
 }
]
