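# Physics asset for the dynamic collar on the male medium Informer body cosmetic, defining its bodies and joints for secondary motion.
# Seven collar boxes (back, three per side) each get tight, mostly swing2 limits against their own small MSTR anchor capsule.
# Anchors link with free angular motion to spine_05, which joins clavicle and upper-arm capsules as body shapes.
# Extra joints tie each side's 03 segment to 02, with twist locked left but free right.
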PhysicsAsset /BRCosmetics/Characters/Player/Male/Medium/Bodies/M_MED_Informer/Meshes/M_MED_Informer_dyn_collar
bodies {
 clavicle_l [{shapeType=Capsule, length=12.5808, radius=6, center=<9.7751 0.8091 -1.2079>, rotation=<rot 90 0 -10.0006>}]
 clavicle_r [{shapeType=Capsule, length=12, radius=6, center=<-6 0 0>, rotation=<rot 90 0 0>}]
 dyn_collar_bk [{shapeType=Box, extentX=5.0473, extentY=10.399, extentZ=1.9852, center=<4 0 1>}]
 dyn_collar_bk_MSTR [{shapeType=Capsule, length=1, radius=1, rotation=<rot 0 0 90>}]
 dyn_collar_l_01 [{shapeType=Box, extentX=5.0473, extentY=10.4553, extentZ=1.9852, center=<4 0 0>}]
 dyn_collar_l_01_MSTR [{shapeType=Capsule, length=1, radius=1, rotation=<rot 0 0 90>}]
 dyn_collar_l_02 [{shapeType=Box, extentX=7.1017, extentY=14.1902, extentZ=2, center=<6.3522 -0.8562 0.1515>, rotation=<rot 5 -20 0>}]
 dyn_collar_l_02_MSTR [{shapeType=Capsule, length=1, radius=1, rotation=<rot 0 0 90>}]
 dyn_collar_l_03 [{shapeType=Box, extentX=6.7671, extentY=10, extentZ=2, center=<5 -1.6848 1.2478>, rotation=<rot 6.3958 -21.9871 -19.6132>}]
 dyn_collar_l_03_MSTR [{shapeType=Capsule, length=1, radius=1, rotation=<rot 0 0 90>}]
 dyn_collar_r_01 [{shapeType=Box, extentX=5.0473, extentY=10.4553, extentZ=1.9852, center=<4 0 0>}]
 dyn_collar_r_01_MSTR [{shapeType=Capsule, length=1, radius=1, rotation=<rot 0 0 90>}]
 dyn_collar_r_02 [{shapeType=Box, extentX=6, extentY=15, extentZ=2, center=<6 0 0>, rotation=<rot 0 0 -10.0006>}]
 dyn_collar_r_02_MSTR [{shapeType=Capsule, length=1, radius=1, rotation=<rot 0 0 90>}]
 dyn_collar_r_03 [{shapeType=Box, extentX=6, extentY=15, extentZ=2, center=<6.0001 0.1847 0.6892>, rotation=<rot 0 0 15.0005>}]
 dyn_collar_r_03_MSTR [{shapeType=Capsule, length=1, radius=1, rotation=<rot 0 0 90>}]
 spine_05 [{shapeType=Capsule, length=17.4597, radius=11.0714, center=<-1.8021 -1.6496 0>}]
 upperarm_l [{shapeType=Capsule, length=16.8362, radius=5.1638, center=<13.0338 -0.3867 0>, rotation=<rot 90 -178.2101 176.7898>}]
 upperarm_r [{shapeType=Capsule, length=15, radius=4, center=<-9.4584 0 0>, rotation=<rot 90 0 0>}]
}
constraints [
 {bone1Name=dyn_collar_bk, bone2Name=dyn_collar_bk_MSTR, swing1LimitDegrees=5, swing2LimitDegrees=15, twistLimitDegrees=5}
 {bone1Name=dyn_collar_l_01, bone2Name=dyn_collar_l_01_MSTR, swing1LimitDegrees=5, swing2LimitDegrees=15, twistLimitDegrees=5}
 {bone1Name=dyn_collar_l_02, bone2Name=dyn_collar_l_02_MSTR, swing1LimitDegrees=5, swing2LimitDegrees=45, twistLimitDegrees=5}
 {bone1Name=dyn_collar_l_03, bone2Name=dyn_collar_l_03_MSTR, swing1LimitDegrees=5, swing2LimitDegrees=30, twistLimitDegrees=5}
 {bone1Name=dyn_collar_r_01, bone2Name=dyn_collar_r_01_MSTR, swing1LimitDegrees=5, swing2LimitDegrees=15, twistLimitDegrees=5}
 {bone1Name=dyn_collar_r_02, bone2Name=dyn_collar_r_02_MSTR, swing1LimitDegrees=5, swing2LimitDegrees=35, twistLimitDegrees=5}
 {bone1Name=dyn_collar_r_03, bone2Name=dyn_collar_r_03_MSTR, swing1LimitDegrees=5, swing2LimitDegrees=35, twistLimitDegrees=5}
 {
  bone1Name=upperarm_l
  bone2Name=clavicle_l
  swing1LimitDegrees=45
  swing1Motion=Free
  swing2LimitDegrees=45
  swing2Motion=Free
  twistLimitDegrees=45
  twistMotion=Free
 }
 {
  bone1Name=upperarm_r
  bone2Name=clavicle_r
  swing1LimitDegrees=45
  swing1Motion=Free
  swing2LimitDegrees=45
  swing2Motion=Free
  twistLimitDegrees=45
  twistMotion=Free
 }
 {
  bone1Name=clavicle_l
  bone2Name=spine_05
  swing1LimitDegrees=45
  swing1Motion=Free
  swing2LimitDegrees=45
  swing2Motion=Free
  twistLimitDegrees=45
  twistMotion=Free
 }
 {
  bone1Name=clavicle_r
  bone2Name=spine_05
  swing1LimitDegrees=45
  swing1Motion=Free
  swing2LimitDegrees=45
  swing2Motion=Free
  twistLimitDegrees=45
  twistMotion=Free
 }
 {
  bone1Name=dyn_collar_bk_MSTR
  bone2Name=spine_05
  swing1LimitDegrees=45
  swing1Motion=Free
  swing2LimitDegrees=45
  swing2Motion=Free
  twistLimitDegrees=45
  twistMotion=Free
 }
 {
  bone1Name=dyn_collar_l_01_MSTR
  bone2Name=spine_05
  swing1LimitDegrees=45
  swing1Motion=Free
  swing2LimitDegrees=45
  swing2Motion=Free
  twistLimitDegrees=45
  twistMotion=Free
 }
 {
  bone1Name=dyn_collar_l_02_MSTR
  bone2Name=spine_05
  swing1LimitDegrees=45
  swing1Motion=Free
  swing2LimitDegrees=45
  swing2Motion=Free
  twistLimitDegrees=45
  twistMotion=Free
 }
 {
  bone1Name=dyn_collar_l_03_MSTR
  bone2Name=spine_05
  swing1LimitDegrees=45
  swing1Motion=Free
  swing2LimitDegrees=45
  swing2Motion=Free
  twistLimitDegrees=45
  twistMotion=Free
 }
 {
  bone1Name=dyn_collar_r_01_MSTR
  bone2Name=spine_05
  swing1LimitDegrees=45
  swing1Motion=Free
  swing2LimitDegrees=45
  swing2Motion=Free
  twistLimitDegrees=45
  twistMotion=Free
 }
 {
  bone1Name=dyn_collar_r_02_MSTR
  bone2Name=spine_05
  swing1LimitDegrees=45
  swing1Motion=Free
  swing2LimitDegrees=45
  swing2Motion=Free
  twistLimitDegrees=45
  twistMotion=Free
 }
 {
  bone1Name=dyn_collar_r_03_MSTR
  bone2Name=spine_05
  swing1LimitDegrees=45
  swing1Motion=Free
  swing2LimitDegrees=45
  swing2Motion=Free
  twistLimitDegrees=45
  twistMotion=Free
 }
 {bone1Name=dyn_collar_l_03, bone2Name=dyn_collar_l_02, swing1LimitDegrees=10, swing2LimitDegrees=15, twistLimitDegrees=45, twistMotion=Locked}
 {bone1Name=dyn_collar_r_03, bone2Name=dyn_collar_r_02, swing1LimitDegrees=5, swing2LimitDegrees=15, twistLimitDegrees=45, twistMotion=Free}
]
