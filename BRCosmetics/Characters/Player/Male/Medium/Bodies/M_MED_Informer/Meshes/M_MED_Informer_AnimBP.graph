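# Post-process style cosmetic AnimBP for the M_MED_Informer body that animates its coat and collar over the pose copied from the attached parent mesh.
# Event update maps thigh, right upper-arm and head rotations to alphas for additive bone offsets that act as pose correctives.
# Two RigidBody nodes then simulate coat (pelvis-relative) and collar (spine_05-relative) using skydiving external force.
# World gravity is overridden, weaker when skydiving without an open parachute.
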AnimBlueprint /BRCosmetics/Characters/Player/Male/Medium/Bodies/M_MED_Informer/Meshes/M_MED_Informer_AnimBP
target_skeleton=&/M_MED_Informer_Skeleton
graphs [
 AnimGraph {nodes [
   n1 = Get thigh_up_l() -> (thigh_up_l:double)
   n2 = "Transform (Modify) Bone - Bone: dyn_coatFront_l_MSTR"(
    ComponentPose=Local To Component(LocalPose=Copy Pose From Mesh() {bUseAttachedParent=true})
    Alpha=@n1
    Translation="-9,0,2"
    Rotation="0,0,-75"
    Scale="1,1,1"
   ) {
    rotationMode=BMM_Additive
    rotationSpace=BCS_BoneSpace
    translationMode=BMM_Additive
    translationSpace=BCS_BoneSpace
    alphaBoolBlend {blendOption=Linear}
    boneToModify {boneName=dyn_coatFront_l_MSTR}
    rotation {roll=-75}
    translation {x=-9, z=2}
   }
   n3 = "Transform (Modify) Bone - Bone: dyn_coatSide_l_MSTR"(ComponentPose=@n2, Alpha=@n1, Translation="-3.5,0,0", Rotation="0,0,0", Scale="1,1,1") {
    rotationMode=BMM_Additive
    rotationSpace=BCS_BoneSpace
    translationMode=BMM_Additive
    translationSpace=BCS_BoneSpace
    alphaBoolBlend {blendOption=Linear}
    boneToModify {boneName=dyn_coatSide_l_MSTR}
    translation {x=-3.5}
   }
   n4 = "Transform (Modify) Bone - Bone: dyn_coatFront_r_MSTR"(ComponentPose=@n3, Alpha=Get thigh_up_r() -> (thigh_up_r:double), Translation="-2.5,0.5,1", Rotation="0,0,50", Scale="1,1,1") {
    rotationMode=BMM_Additive
    rotationSpace=BCS_BoneSpace
    translationMode=BMM_Additive
    translationSpace=BCS_BoneSpace
    alphaBoolBlend {blendOption=Linear}
    boneToModify {boneName=dyn_coatFront_r_MSTR}
    rotation {roll=50}
    translation=<-2.5 0.5 1>
   }
   n5 = Get arm_up_r() -> (arm_up_r:double)
   n6 = "Transform (Modify) Bone - Bone: dyn_collar_r_03_MSTR"(ComponentPose=@n4, Alpha=@n5, Translation="0,0,0", Rotation="50,15,0", Scale="1,1,1") {
    rotationMode=BMM_Additive
    rotationSpace=BCS_BoneSpace
    scaleMode=BMM_Additive
    scaleSpace=BCS_BoneSpace
    translationMode=BMM_Additive
    translationSpace=BCS_BoneSpace
    alphaBoolBlend {blendOption=Linear}
    boneToModify {boneName=dyn_collar_r_03_MSTR}
    rotation {pitch=50, yaw=15}
   }
   n7 = "Transform (Modify) Bone - Bone: dyn_collar_r_02_MSTR"(ComponentPose=@n6, Alpha=@n5, Translation="0,0,0", Rotation="35,0.000,0.000", Scale="1,1,1") {
    rotationMode=BMM_Additive
    rotationSpace=BCS_BoneSpace
    alphaBoolBlend {blendOption=Linear}
    boneToModify {boneName=dyn_collar_r_02_MSTR}
    rotation {pitch=35}
   }
   n8 = Get head_rot_back() -> (head_rot_back:double)
   n9 = "Transform (Modify) Bone - Bone: dyn_collar_bk_MSTR"(ComponentPose=@n7, Alpha=@n8, Translation="0,0,0", Rotation="-10.000,-0.000,-5.000", Scale="1,1,1") {
    rotationMode=BMM_Additive
    rotationSpace=BCS_BoneSpace
    translationMode=BMM_Additive
    translationSpace=BCS_BoneSpace
    alphaBoolBlend {blendOption=Linear}
    boneToModify {boneName=dyn_collar_bk_MSTR}
    rotation=<rot -10 0 -5.0001>
   }
   n10 = "Transform (Modify) Bone - Bone: dyn_collar_r_01_MSTR"(ComponentPose=@n9, Alpha=@n8, Translation="0,0,0", Rotation="-30,0,10.000", Scale="1,1,1") {
    rotationMode=BMM_Additive
    rotationSpace=BCS_BoneSpace
    translationMode=BMM_Additive
    translationSpace=BCS_BoneSpace
    alphaBoolBlend {blendOption=Linear}
    boneToModify {boneName=dyn_collar_r_01_MSTR}
    rotation {pitch=-30, roll=10}
   }
   n11 = "Transform (Modify) Bone - Bone: dyn_collar_l_01_MSTR"(ComponentPose=@n10, Alpha=@n8, Translation="0,0,0", Rotation="-20,0,-15", Scale="1,1,1") {
    rotationMode=BMM_Additive
    rotationSpace=BCS_BoneSpace
    translationMode=BMM_Additive
    translationSpace=BCS_BoneSpace
    alphaBoolBlend {blendOption=Linear}
    boneToModify {boneName=dyn_collar_l_01_MSTR}
    rotation {pitch=-20, roll=-15}
   }
   n12 = Break Fort Anim Input Skydiving External Force(FortAnimInput_SkydivingExternalForce=Get SkydivingExternalForceInput()) -> (
    bUseSkydivingVectorForce=~
    SkydivingVectorForce=~
    HeadToPelvisDirection=~
    FloatingMultiplier=~
    FloatingAdditive=~
    DivingMultiplier=~
    DivingAdditive=~
    ParachutingMultiplier=~
    ParachutingAdditive=~
   )
   n13 = Get gravity_override() -> (gravity_override:Vector)
   n14 = RigidBody!(ComponentPose=@n11, Alpha, OverrideWorldGravity:Vector=@n13, ExternalForce=@n12.SkydivingVectorForce) -> (Pose) {
    bClampLinearTranslationLimitToRefPose=true
    bOverrideWorldGravity=true
    overridePhysicsAsset=&/M_MED_Informer_dyn_coat
    simulationSpace=BaseBoneSpace
    alphaBoolBlend {blendOption=Linear}
    baseBoneRef {boneName=pelvis}
    componentAppliedLinearAccClamp=<1500 1500 1500>
    componentLinearAccScale=<0.35 0.35 0.35>
    componentLinearVelScale=<0.2 0.2 0.2>
    overrideWorldGravity {z=-980}
   }
   n15 = RigidBody(ComponentPose=@n14, OverrideWorldGravity=@n13, ExternalForce=@n12.SkydivingVectorForce) {
    bClampLinearTranslationLimitToRefPose=true
    bOverrideWorldGravity=true
    overridePhysicsAsset=&/M_MED_Informer_dyn_collar
    simulationSpace=BaseBoneSpace
    alphaBoolBlend {blendOption=Linear}
    baseBoneRef {boneName=spine_05}
    componentAppliedLinearAccClamp=<1500 1500 1500>
    componentLinearAccScale=<0.2 0.2 0.2>
    componentLinearVelScale=<0.2 0.2 0.2>
    overrideWorldGravity {z=-980}
   }
   Output Pose(Result=Component To Local(ComponentPose=@n15)) {layerGroup=None}
  ]}
 EventGraph {nodes [
   Event BlueprintUpdateAnimation()
   > n1 = Sequence!(execute) -> (then_0, then_1, then_2:Exec, then_3:Exec, then_4:Exec)
   > n2 = Branch(Condition=Get bIsSkyDiving())
   > n3 = Branch(Condition=Get bIsParachuteOpened())
   Set gravity_override(execute=[@n3.then, @n2.else], gravity_override="0, 0,-980")
   Set gravity_override(execute=@n3.else, gravity_override="0, 0,-300")
   n4 = BreakTransform(InTransform=GetDeltaTransformFromRefPose(self=GetOwnerSkeletalMeshComponent(), BoneName=thigh_r, BaseName=pelvis))
   Set thigh_up_r(execute=@n1.then_1, thigh_up_r:double=MapRangeClamped(Value=BreakRotator(InRot=@n4.Rotation).Yaw, InRangeA=-35, InRangeB=-120)) -> (Output_Get:double)
   n5 = BreakTransform(InTransform=GetDeltaTransformFromRefPose(self=GetOwnerSkeletalMeshComponent(), BoneName=thigh_l, BaseName=pelvis))
   Set thigh_up_l(execute=@n1.then_2, thigh_up_l:double=MapRangeClamped(Value=BreakRotator(InRot=@n5.Rotation).Yaw, InRangeA=-35, InRangeB=-120)) -> (Output_Get:double)
   n6 = BreakTransform(InTransform=GetDeltaTransformFromRefPose(self=GetOwnerSkeletalMeshComponent(), BoneName=upperarm_r, BaseName=spine_05))
   Set arm_up_r(execute=@n1.then_3, arm_up_r:double=MapRangeClamped(Value=BreakRotator(InRot=@n6.Rotation).Yaw, InRangeA=30, InRangeB=100)) -> (Output_Get:double)
   n7 = BreakTransform(InTransform=GetDeltaTransformFromRefPose(self=GetOwnerSkeletalMeshComponent(), BoneName=head, BaseName=spine_05))
   > Set head_rot_back(head_rot_back:double=MapRangeClamped(Value=BreakRotator(InRot=@n7.Rotation).Yaw, InRangeA=-10, InRangeB=-40)) -> (Output_Get:double)
  ]}
]
variables [
 arm_up_r {type=number}
 gravity_override {type=Vector, default=<0 0 -980>}
 head_rot_back {type=number}
 thigh_up_l {type=number}
 thigh_up_r {type=number}
]
