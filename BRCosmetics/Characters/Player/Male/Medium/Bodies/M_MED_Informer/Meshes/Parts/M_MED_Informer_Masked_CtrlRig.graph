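# Control Rig for the masked M_MED_Informer character that shifts and tilts the glasses bone (`dyn_galsses`) in response to cheek movement.
# Left and right inner cheek Z positions, relative to `faceAttach`, are remapped to clamped weights.
# These weights drive two chained additive-local bone offsets with mirrored rotations and a shared small translation, so each cheek tilts the glasses oppositely.
# A similar nose-bridge remap exists but its output is unconnected.
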ControlRigBlueprint /BRCosmetics/Characters/Player/Male/Medium/Bodies/M_MED_Informer/Meshes/Parts/M_MED_Informer_Masked_CtrlRig
graphs [
 RigVMFunctionLibrary {links [], nodes []}
 RigVMModel {nodes [
   n1 = &RigUnit_BeginExecution()
   /Script/ControlRig.RigUnit_ModifyBoneTransforms(
    BoneToModify=((Bone="dyn_galsses",Transform=(Rotation=(X=0,Y=0,Z=0,W=1),Translation=(X=0,Y=0.2,Z=0.2),Scale3D=(X=1,Y=1,Z=1))))
    Mode=AdditiveLocal
    WeightMaximum=1
    WeightMinimum=0
    ExecutePin=@n1.ExecutePin
    Weight=&RigVMFunction_MathFloatRemap(
     SourceMaximum=3.2
     SourceMinimum=2.12
     TargetMinimum=0
     bClamp=true
     Value=&RigUnit_GetRelativeBoneTransform(Bone=L_cheek_inner, Space=faceAttach).Transform.Translation.Z
    ).Result
    BoneToModify.0.Transform.Rotation=/Script/RigVM.RigVMFunction_MathQuaternionFromRotator(Rotator=(Pitch=1,Yaw=0,Roll=-0.5)) {is=struct RigVMUnitNode}.Result
   ) {is=struct RigVMUnitNode}
   > &RigUnit_ModifyBoneTransforms(
    BoneToModify=((Bone="dyn_galsses",Transform=(Rotation=(X=0,Y=0,Z=0,W=1),Translation=(X=0,Y=0.2,Z=0.2),Scale3D=(X=1,Y=1,Z=1))))
    Mode=AdditiveLocal
    WeightMaximum=1
    WeightMinimum=0
    Weight=&RigVMFunction_MathFloatRemap(
     SourceMaximum=3.2
     SourceMinimum=2.12
     TargetMinimum=0
     bClamp=true
     Value=&RigUnit_GetRelativeBoneTransform(Bone=R_cheek_inner, Space=faceAttach).Transform.Translation.Z
    ).Result
    BoneToModify.0.Transform.Rotation=&RigVMFunction_MathQuaternionFromRotator(Rotator=(Pitch=-1,Yaw=0,Roll=0.5)).Result
   )
   &RigVMFunction_MathFloatRemap(
    SourceMaximum=3.2
    SourceMinimum=3.6
    TargetMinimum=0
    bClamp=true
    Value=&RigUnit_GetRelativeBoneTransform(Bone=C_nose_bridge, Space=faceAttach).Transform.Translation.Z
   )
  ]}
]
hierarchy [
 root {loc=<0 0 0>, rot=<0 0 0 1>}
 pelvis {parent=root, loc=<0 -0.7071 93.1191>, rot=<0.0127 0.707 -0.0127 -0.707>}
 spine_01 {parent=pelvis, loc=<6.7494 -0.6388 0>, rot=<0 0 0.034 -0.9994>}
 spine_02 {parent=spine_01, loc=<7.2518 -0.035 0>, rot=<0 0 0.0238 -0.9997>}
 spine_03 {parent=spine_02, loc=<6.0104 0 0>, rot=<0 0 0.0149 -0.9999>}
 spine_04 {parent=spine_03, loc=<8.1024 0 0>, rot=<0 0 -0.0132 -0.9999>}
 spine_05 {parent=spine_04, loc=<10.0146 0 0>, rot=<0 0 -0.0015 -1>}
 neck_01 {parent=spine_05, loc=<11.3138 0.1529 0>, rot=<0 0 -0.2003 -0.9797>}
 neck_02 {parent=neck_01, loc=<4.5627 0 0>, rot=<0 0 0.0504 -0.9987>}
 head {parent=neck_02, loc=<5.3914 0 0>, rot=<0 0 0.115 -0.9934>}
 dyn_galsses {parent=head, loc=<5.5174 0.2487 0>, rot=<0.0029 -0.7071 -0.0029 -0.7071>}
 faceAttach {parent=head, loc=<-0.8283 -0.1832 0.0002>, rot=<0.0082 -0.7071 -0.0082 -0.7071>, scale=<0.89 0.89 0.89>}
 L_cheek_inner {parent=faceAttach, loc=<4.1068 8.6623 2.1203>, rot=<-0.0707 -0.2698 0.3184 -0.906>}
 C_nose_bridge {parent=faceAttach, loc=<-0.0258 13.0809 3.5578>, rot=<-0.2218 0.0069 0.0019 -0.9751>}
 R_cheek_inner {parent=faceAttach, loc=<-4.0938 8.6708 2.113>, rot=<-0.0719 0.2728 -0.3179 -0.9052>}
]
