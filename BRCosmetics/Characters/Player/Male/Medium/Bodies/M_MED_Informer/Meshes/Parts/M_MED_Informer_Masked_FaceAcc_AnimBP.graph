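# Animation Blueprint for the masked face accessory mesh of the M_MED_Informer male player cosmetic body.
# It copies the pose and curves from the attached parent mesh and feeds them into the M_MED_Informer_Masked Control Rig.
# Each update sets an isLOD0 value that drives the rig's alpha, so the rig applies only at LOD 0.
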AnimBlueprint /BRCosmetics/Characters/Player/Male/Medium/Bodies/M_MED_Informer/Meshes/Parts/M_MED_Informer_Masked_FaceAcc_AnimBP
target_skeleton=&/M_MED_Informer_Masked_FaceAcc_Skeleton
graphs [
 AnimGraph {nodes [
   n1 = Control Rig(Source=Copy Pose From Mesh() {bCopyCurves=true, bUseAttachedParent=true}, Alpha=Get isLOD0() -> (isLOD0:double)) {
    alphaBoolBlend {blendOption=Linear}
    controlRigAssetReference {blueprintRigClass=&/M_MED_Informer_Masked_CtrlRig.M_MED_Informer_Masked_CtrlRig_C}
   }
   Output Pose(Result=@n1) {layerGroup=None}
  ]}
 EventGraph {nodes [
   Event BlueprintUpdateAnimation()
   > n1 = Branch(Condition="Equal (Integer)"(A=Get CurrentLOD(), B=0) {is=title K2Node_CallFunction})
   > Set isLOD0(isLOD0:double=1) -> (Output_Get:double)
   Set isLOD0(execute=@n1.else, isLOD0=0)
   TryGetPawnOwner()
  ]}
]
variables [isLOD0 {type=number}]
